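# Niagara system for the Pulsed Energy vehicle boost cosmetic, emitting from skeletal mesh sockets split by port count.
# A ribbon emitter draws the trail, and a burst sprite emitter holds looping core glows.
# A merged mesh/sprite emitter uses spawn groups as visibility tags and array indices to vary pulse color, size, lifetime and velocity, with extra color and dynamic parameter bindings.
# User parameters drive painted-color variants, intensities and per-car offset.
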NiagaraSystem /VehicleCosmetics/Boosts/Boost_PulsedEnergy/FX/NS_Boost_PulsedEnergy_Opt
dependencies {
 usedDataInterfaces [
  &NiagaraDataInterfaceVectorCurve
  &NiagaraDataInterfaceCurve
  &NiagaraDataInterfaceArrayFloat
  &NiagaraDataInterfaceArrayColor
  &NiagaraDataInterfaceArrayFloat3
  &NiagaraDataInterfaceSkeletalMesh
 ]
 usedDynamicInputs []
 usedModules [
  &self:SystemSpawnScript.SystemScriptSource.SystemScriptGraph.NiagaraNodeAssignment_0.SetVariables_#1
  &SystemState
 ]
}
emitters {
 NE_VC_Boost_Ribbon_Parent {
  emitterUpdateScript [
   &EmitterState()
   &VC_FilteredSocketCount()
   &VC_RateQualityHandler(Epic Spawn Rate=30, High Spawn Rate=30, Low Spawn Rate=30, Use Custom Rate?=true)
   &VC_SpawnRibbon_v1(Number of Ribbons=$Emitter.PortCount, Spawn Time=-1, SpawnRate=$Emitter.Spawn Rate)
  ]
  particleSpawnScript [
   &VC_PaintedVariantsOverride() {enabled=false}
   &VC_InitializeRibbon_v1(
    Color=&MakeCustomLinearColorFromBool(Bool=$User.bIsPainted, True Color=$User.PaintedColor, False Color=<rgba 1 1 1 1>)
    Custom Offset=true
    Enable Skeletal Position=true
    Lifetime=0.6
    Local Vector Offset=&AddVectorToPosition(Position=<-25 0 0>, Vector=$User.DTCarOffset)
    Offset Amount=1
    Ribbon Width=62
   )
   &AddVelocity(Velocity=<-2000 0 0>) {module=AddVelocity001}
  ]
  particleUpdateScript [
   &ParticleState()
   &InheritSourceMovement(
    Applied Position Scale=&VectorFromCurve(VectorCurve=DataInterface() {propertyValues {
       XCurve {keys [<at 0 1 RCIM_Cubic>, {interpMode=RCIM_Cubic, leaveTangent=-1.655, time=0.7, value=0}]}
       YCurve {keys [<at 0 1 RCIM_Cubic>, {interpMode=RCIM_Cubic, leaveTangent=-1.655, time=0.7, value=0}]}
       ZCurve {keys [<at 0 1 RCIM_Cubic>, {interpMode=RCIM_Cubic, leaveTangent=-1.655, time=0.7, value=0}]}
      }})
    ApplyForce=false
    ApplyPosition=true
    ApplyPositionFalloff=false
   )
   &SolveForcesAndVelocity(Clamp Velocity=false, Limit Acceleration=false)
   &VC_UpdateRibbon_v1(Enable Skeletal Position=true)
   &ScaleRibbonWidth(Ribbon Width Scale=&FloatFromCurve(FloatCurve=DataInterface() {propertyValues {Curve {keys [<at 0 1.48>,<at 1 0.5>]}}}))
   &DynamicMaterialParameters(
    Index 0 Param 1=$User.RibbonCoreIntensity
    Index 0 Param 2=$User.RibbonMidIntensity
    Index 0 Param 3=$User.RibbonOutIntensity
    Index 0 Param 4=&MakeCustomFloatFromBool(Bool=$User.bIsPainted, True Float=1, False Float=0)
    Module.Index0Param0Write=true
    Module.Index0Param1Write=true
    Module.Index0Param2Write=true
    Module.Index0Param3Write=true
   )
  ]
  rendererClasses [&NiagaraRibbonRendererProperties]
  renderers [{rendererClass=&NiagaraRibbonRendererProperties}]
 }
 NE_VC_Boost_Sprite_Parent {
  emitterUpdateScript [
   &EmitterState()
   &VC_FilteredSocketCount()
   &SpawnRate(SpawnRate=20, Use Spawn Probability=false)
   &SpawnRate(SpawnGroup=3, SpawnRate=20, Use Spawn Probability=false) {module=SpawnRate005}
   &SpawnRate(SpawnGroup=1, SpawnRate=20, Use Spawn Probability=false) {module=SpawnRate001}
   &SpawnRate(SpawnGroup=4, SpawnRate=20, Use Spawn Probability=false) {module=SpawnRate004}
   &SpawnRate(SpawnGroup=2, SpawnRate=3, Use Spawn Probability=false) {module=SpawnRate002}
   &SpawnRate(SpawnGroup=5, SpawnRate=3, Use Spawn Probability=false) {module=SpawnRate003}
  ]
  particleSpawnScript [
   &VC_PaintedVariantsOverride() {enabled=false}
   /SetVariables(Particles.VisibilityTag=&SpawnGroup()) {isSetParameters=true}
   /VehicleCosmetics/FX/Niagara/Modules/MergeEmitters/VC_ExtraBindingColors(
    ExtraColors=1
    R1 Color:LinearColor=&MakeCustomLinearColorFromBool(Bool=$User.bIsPainted, True Color=$User.PaintedColor, False Color=<rgba 0.0759 0.7519 1 1>)
   )
   &InitializeParticle(
    Color=&MakeCustomLinearColorFromBool(
     Bool=$User.bIsPainted
     True Color=$User.PaintedColor
     False Color=&SelectLinearColorFromArray(
      Array Sampling Mode=::NewEnumerator1
      Direct Array Index=&SpawnGroup()
      Direct Array Mode=::NewEnumerator1
      Color Selection Array=DataInterface() {propertyValues {ColorData [<rgba 1 1 1 1>, <rgba 0.0066 1 0.9158 1>, <rgba 0.0759 0.7519 1 1>]}}
      Random Seed=?
     )
    )
    Lifetime=&SelectFloatFromArray(
     Direct Array Index=&SpawnGroup()
     Direct Array Mode=::NewEnumerator1
     Float Selection Array=DataInterface() {propertyValues {FloatData [0.3,0.3,0.7]}}
    )
    Mesh Scale=&SelectVectorFromArray(
     Direct Array Index=&SpawnGroup()
     Direct Array Mode=::NewEnumerator1
     Vector Selection Array=DataInterface() {propertyValues {FloatData [<0.14 0.11 0.11>, <0.11 0.09 0.09>, <1.7 1.7 1.35>]}}
    )
    Mesh Scale Mode=::NewEnumerator4
    Sprite Rotation Mode=::NewEnumerator2
    Sprite Size Mode=::NewEnumerator3
    UsePositionOffset=false
    Write Material Random=true
   )
   &PartitionParticles(Number Of Partitions=$Emitter.PortCount)
   &SkeletalMeshLocation(
    Bone / Socket Index=$Particles.PartitionParticles.Partition
    Bone / Socket Sampling Mode=::NewEnumerator5
    Interpolate Mesh Sampling=false
    MaskSpawnGroup=false
    Mesh Sampling Type=::NewEnumerator6
    Orientation Sampling=::NewEnumerator1
    Sampled Position Offset=&AddVectorToPosition(
     Position=<-80 0 0>
     Vector=&AddVectorToPosition(
      Position=&SelectVectorFromArray(
       Direct Array Index=&SpawnGroup()
       Direct Array Mode=::NewEnumerator1
       Vector Selection Array=DataInterface() {propertyValues {FloatData [<-5.5 0 0>, <-9.5 0 0>, <0 0 0>]}}
      )
      Vector=$User.DTCarOffset
     )
    )
    Sampled Position Offset Coordinate Space:ENiagaraSkelMeshTransforms=::NewEnumerator2
    Skeletal Mesh=$System.SkeletalMesh
   )
   &InitialMeshOrientation(
    Mesh Orientation Mode=::NewEnumerator7
    Rotation=&RandomRangeVector(
     Minimum=&SelectVectorFromArray(
      Direct Array Index=&SpawnGroup()
      Direct Array Mode=::NewEnumerator1
      Vector Selection Array=DataInterface() {propertyValues {FloatData [<0 0 0>, <0 0 0>, <0 0 0>]}}
     )
     Maximum=&SelectVectorFromArray(
      Direct Array Index=&SpawnGroup()
      Direct Array Mode=::NewEnumerator1
      Vector Selection Array=DataInterface() {propertyValues {FloatData [<2 0 0>, <1 0 0>, <0 0 0>]}}
     )
    )
    Rotation Coordinate Space=::NewEnumerator2
    Use Rotation Vector=true
   )
   &AddVelocity(Velocity=&SelectVectorFromArray(
     Direct Array Index=&SpawnGroup()
     Direct Array Mode=::NewEnumerator1
     Vector Selection Array=DataInterface() {propertyValues {FloatData [<-19 0 0>, <-15 0 0>, <-160 0 0>]}}
    ))
  ]
  particleUpdateScript [
   &ParticleState()
   &ScaleMeshSize(Scale Factor=&MaskVectorBySpawnGroup(
     Masked Vector=&VectorFromCurve(VectorCurve=DataInterface() {propertyValues {XCurve {keys [<at 0 0.5>,<at 1 3>]}, YCurve {keys [<at 0 1>,<at 1 1.8>]}, ZCurve {keys [<at 0 1>,<at 1 1.8>]}}})
     Passthrough Vector=<1 1 1>
    ))
   &ScaleMeshSize(Scale Factor=&MaskVectorBySpawnGroup(
     Masked Vector=&VectorFromCurve(VectorCurve=DataInterface() {propertyValues {XCurve {keys [<at 0 0.5>,<at 1 3>]}, YCurve {keys [<at 0 1>,<at 1 1.8>]}, ZCurve {keys [<at 0 1>,<at 1 1.8>]}}})
     Passthrough Vector=<1 1 1>
     Spawn Group Mask=3
    )) {module=ScaleMeshSize005}
   &ScaleMeshSize(Scale Factor=&MaskVectorBySpawnGroup(
     Masked Vector=&VectorFromCurve(VectorCurve=DataInterface() {propertyValues {XCurve {keys [<at 0 0.5>,<at 1 3>]}, YCurve {keys [<at 0 1>,<at 1 1.8>]}, ZCurve {keys [<at 0 1>,<at 1 1.8>]}}})
     Passthrough Vector=<1 1 1>
     Spawn Group Mask=1
    )) {module=ScaleMeshSize004}
   &ScaleMeshSize(Scale Factor=&MaskVectorBySpawnGroup(
     Masked Vector=&VectorFromCurve(VectorCurve=DataInterface() {propertyValues {XCurve {keys [<at 0 0.5>,<at 1 3>]}, YCurve {keys [<at 0 1>,<at 1 1.8>]}, ZCurve {keys [<at 0 1>,<at 1 1.8>]}}})
     Passthrough Vector=<1 1 1>
     Spawn Group Mask=4
    )) {module=ScaleMeshSize003}
   &ScaleMeshSize(Scale Factor=&MaskVectorBySpawnGroup(
     Masked Vector=&VectorFromCurve(VectorCurve=DataInterface() {propertyValues {
        XCurve {keys [
          {interpMode=RCIM_Cubic, leaveTangentWeight=0.3333, tangentMode=RCTM_User, time=0, value=0.5}
          {arriveTangent=0.0289, tangentMode=RCTM_User, time=0.3, value=1.2}
          <at 1 0.6>
         ]}
        YCurve {keys [
          {interpMode=RCIM_Cubic, leaveTangentWeight=0.3333, tangentMode=RCTM_User, time=0, value=0.5}
          {arriveTangent=0.0289, tangentMode=RCTM_User, time=0.3, value=1.2}
          <at 1 0.6>
         ]}
        ZCurve {keys [
          {interpMode=RCIM_Cubic, leaveTangentWeight=0.3333, tangentMode=RCTM_User, time=0, value=0.5}
          {arriveTangent=0.0289, tangentMode=RCTM_User, time=0.3, value=1.2}
          <at 1 0.6>
         ]}
       }})
     Passthrough Vector=<1 1 1>
     Spawn Group Mask=2
    )) {module=ScaleMeshSize001}
   &ScaleMeshSize(Scale Factor=&MaskVectorBySpawnGroup(
     Masked Vector=&VectorFromCurve(VectorCurve=DataInterface() {propertyValues {
        XCurve {keys [
          {interpMode=RCIM_Cubic, leaveTangentWeight=0.3333, tangentMode=RCTM_User, time=0, value=0.5}
          {arriveTangent=0.0289, tangentMode=RCTM_User, time=0.3, value=1.2}
          <at 1 0.6>
         ]}
        YCurve {keys [
          {interpMode=RCIM_Cubic, leaveTangentWeight=0.3333, tangentMode=RCTM_User, time=0, value=0.5}
          {arriveTangent=0.0289, tangentMode=RCTM_User, time=0.3, value=1.2}
          <at 1 0.6>
         ]}
        ZCurve {keys [
          {interpMode=RCIM_Cubic, leaveTangentWeight=0.3333, tangentMode=RCTM_User, time=0, value=0.5}
          {arriveTangent=0.0289, tangentMode=RCTM_User, time=0.3, value=1.2}
          <at 1 0.6>
         ]}
       }})
     Passthrough Vector=<1 1 1>
     Spawn Group Mask=5
    )) {module=ScaleMeshSize002}
   &SolveForcesAndVelocity(Clamp Velocity=false, Limit Acceleration=false)
   &DynamicMaterialParameters(
    Index 0 Param 1=&FloatFromCurve(FloatCurve=DataInterface() {propertyValues {Curve {keys [<at 0 1>,<at 1 0>]}}})
    Index 0 Param 2=1
    Index 0 Param 3=$User.InColorIntensity
    Index 0 Param 4=$User.OutColorIntensity
    Index 1 Param 1=&MakeCustomFloatFromBool(Bool=$User.bIsPainted, True Float=1, False Float=0)
    Module.Index0Param0Write=true
    Module.Index0Param1Write=true
    Module.Index0Param2Write=true
    Module.Index0Param3Write=true
    Module.Index1Param0Write=true
    Module.Index1Param1Write=false
    Module.Index1Param2Write=false
    Module.Index1Param3Write=false
    Write Parameter Index 1=true
   )
   /SetVariables(
    Particles.R1 Color:LinearColor=&VC_DynamicScaleColor(
     Color to Scale:LinearColor=$Particles.Initial.R1 Color
     RGB Scale:Vector3f=&MakeCustomVectorFromBool(
      Bool=$User.bIsPainted
      True Vector=&MakeVector(X=$User.RingsIntensityR, Y=$User.RingsIntensityG, Z=$User.RingsIntensityB)
      False Vector=<2.5 2.5 2.5>
     )
     Alpha Scale:NiagaraFloat=&FloatFromCurve(FloatCurve=DataInterface() {propertyValues {Curve {keys [<at 0 0.8>,<at 1 0>]}}})
    )
    Particles.Rings_M3_Dynamic:Vector4f=&MakeVector4(
     X=&MakeCustomFloatFromBool(Bool=$User.bIsPainted, True Float=$User.RingsOpacity, False Float=1)
     Y=0
     Z=0
     W=0
    )
    Particles.SplatImpact_M2_Dynamic:Vector4f=&MakeVector4(
     X=&FloatFromCurve(FloatCurve=DataInterface() {propertyValues {Curve {keys [<at 0 1>,<at 1 0>]}}})
     Y=1
     Z=$User.SingleColorIntensity
     W=0
    )
   ) {isSetParameters=true}
   &ScaleColor(
    Scale Alpha=&MaskFloatBySpawnGroup(
     Masked Float=&FloatFromCurve(FloatCurve=DataInterface() {propertyValues {Curve {keys [<at 0 1>,<at 0.8 0.3>,<at 1 0>]}}})
     Passthrough Float=1
     Spawn Group Mask=1
    )
    Scale RGB=&MaskVectorBySpawnGroup(Masked Vector=<2 2 2>, Passthrough Vector=<2 2 2>, Spawn Group Mask=1)
    ScaleA=true
    ScaleRGB=true
   ) {enabled=false}
   &ScaleColor(
    Scale Alpha=&MaskFloatBySpawnGroup(
     Masked Float=&FloatFromCurve(FloatCurve=DataInterface() {propertyValues {Curve {keys [<at 0 1>,<at 0.8 0.3>,<at 1 0>]}}})
     Passthrough Float=1
     Spawn Group Mask=4
    )
    Scale RGB=&MaskVectorBySpawnGroup(Masked Vector=<2 2 2>, Passthrough Vector=<2 2 2>, Spawn Group Mask=4)
    ScaleA=true
    ScaleRGB=true
   ) {enabled=false, module=ScaleColor003}
   &ScaleColor(
    Scale Alpha=&MaskFloatBySpawnGroup(
     Masked Float=&FloatFromCurve(FloatCurve=DataInterface() {propertyValues {Curve {keys [<at 0 0.8>,<at 1 0>]}}})
     Passthrough Float=1
     Spawn Group Mask=2
    )
    Scale RGB=&MaskVectorBySpawnGroup(
     Masked Vector=&MakeCustomVectorFromBool(
      Bool=$User.bIsPainted
      True Vector=&MakeVector(X=$User.RingsIntensityR, Y=$User.RingsIntensityG, Z=$User.RingsIntensityB)
      False Vector=<2.5 2.5 2.5>
     )
     Passthrough Vector=<1 1 1>
     Spawn Group Mask=2
    )
    ScaleA=true
    ScaleRGB=true
   ) {enabled=false, module=ScaleColor001}
   &ScaleColor(
    Scale Alpha=&MaskFloatBySpawnGroup(
     Masked Float=&FloatFromCurve(FloatCurve=DataInterface() {propertyValues {Curve {keys [<at 0 0.8>,<at 1 0>]}}})
     Passthrough Float=1
     Spawn Group Mask=5
    )
    Scale RGB=&MaskVectorBySpawnGroup(
     Masked Vector=&MakeCustomVectorFromBool(
      Bool=$User.bIsPainted
      True Vector=&MakeVector(X=$User.RingsIntensityR, Y=$User.RingsIntensityG, Z=$User.RingsIntensityB)
      False Vector=<2.5 2.5 2.5>
     )
     Passthrough Vector=<1 1 1>
     Spawn Group Mask=5
    )
    ScaleA=true
    ScaleRGB=true
   ) {enabled=false, module=ScaleColor002}
  ]
  rendererClasses [&NiagaraMeshRendererProperties, &NiagaraSpriteRendererProperties]
  renderers [
   {rendererClass=&NiagaraMeshRendererProperties, rendererIndex=0}
   {rendererClass=&NiagaraMeshRendererProperties, rendererIndex=1}
   {rendererClass=&NiagaraMeshRendererProperties, rendererIndex=2}
   {rendererClass=&NiagaraMeshRendererProperties, rendererIndex=3}
   {rendererClass=&NiagaraMeshRendererProperties, rendererIndex=4}
   {rendererClass=&NiagaraMeshRendererProperties, rendererIndex=5}
   {rendererClass=&NiagaraSpriteRendererProperties, rendererIndex=6}
  ]
 }
 NE_VC_Boost_Sprite_Parent001 {
  emitterUpdateScript [
   &EmitterState(
    Inactive Response=::NewEnumerator1
    Life Cycle Mode=::NewEnumerator1
    Loop Behavior=::NewEnumerator1
    Loop Duration Mode=::NewEnumerator1
    UseLoopDelay=false
   )
   &VC_FilteredSocketCount()
   &SpawnBurst_Instantaneous(Use Spawn Probability=false, UseLoopCountLimit=false)
   &SpawnBurst_Instantaneous(Spawn Group=1, Use Spawn Probability=false, UseLoopCountLimit=false) {module=SpawnBurst_Instantaneous001}
   &SpawnRate(SpawnRate=50, Use Spawn Probability=false) {enabled=false}
  ]
  particleSpawnScript [
   &VC_PaintedVariantsOverride() {enabled=false}
   &InitializeParticle(
    Color=&MakeCustomLinearColorFromBool(
     Bool=$User.bIsPainted
     True Color=&Multiply_LinearColorByFloat(Linear Color=$User.PaintedColor, Float=$User.CoreIntensity)
     False Color=&Multiply_LinearColorByFloat(Linear Color=<rgba 1 1 1 1>, Float=$User.CoreIntensity)
    )
    Sprite Rotation Mode=::NewEnumerator2
    Sprite Size Mode=::NewEnumerator3
    Uniform Sprite Size=30
    UsePositionOffset=false
    Write Material Random=true
   )
   &PartitionParticles(Number Of Partitions=$Emitter.PortCount)
   &SkeletalMeshLocation(
    Bone / Socket Index=$Particles.PartitionParticles.Partition
    Bone / Socket Sampling Mode=::NewEnumerator5
    Interpolate Mesh Sampling=false
    MaskSpawnGroup=false
    Mesh Sampling Type=::NewEnumerator6
    Orientation Sampling=::NewEnumerator1
    Sampled Position Offset=&AddVectorToPosition(Position=<-80 0 0>, Vector=&AddVectorToPosition(Position=<-10.5 0 0>, Vector=$User.DTCarOffset))
    Sampled Position Offset Coordinate Space:ENiagaraSkelMeshTransforms=::NewEnumerator2
    Skeletal Mesh=$System.SkeletalMesh
   )
   &AddVelocity(Velocity=<-1500 0 0>) {enabled=false}
  ]
  particleUpdateScript [
   &ParticleState(Kill Particles When Lifetime Has Elapsed=false, Loop Particles Lifetime=true)
   &SolveForcesAndVelocity(Clamp Velocity=false, Limit Acceleration=false)
   &DynamicMaterialParameters(Module.Index0Param0Write=true, Module.Index0Param1Write=false, Module.Index0Param2Write=false, Module.Index0Param3Write=false)
   &ScaleColor(Scale RGB=<2.5 2.5 2.5>, ScaleA=false, ScaleRGB=true)
  ]
  rendererClasses [&NiagaraSpriteRendererProperties]
  renderers [
   {rendererClass=&NiagaraSpriteRendererProperties}
   {rendererClass=&NiagaraSpriteRendererProperties, rendererIndex=1}
  ]
 }
}
summary {
 systemName=NS_Boost_PulsedEnergy_Opt
 emitters [
  {emitterName=NE_VC_Boost_Ribbon_Parent, rendererClasses [&NiagaraRibbonRendererProperties]}
  {
   emitterName=NE_VC_Boost_Sprite_Parent
   rendererClasses [&NiagaraMeshRendererProperties, &NiagaraSpriteRendererProperties]
  }
  {emitterName=NE_VC_Boost_Sprite_Parent001, rendererClasses [&NiagaraSpriteRendererProperties]}
 ]
 userVariables [
  User.bIsFrontendPreview {type=NiagaraBool}
  User.bIsPainted {type=NiagaraBool}
  User.ColorVariantID {type=NiagaraInt32}
  User.CoreIntensity {type=NiagaraFloat, defaultValue=1, description=""}
  User.DTCarOffset {type=Vector3f, description="", defaultValue {struct=Vector3f, value=<0 0 0>}}
  User.InColorIntensity {type=NiagaraFloat, defaultValue=1, description=""}
  User.OutColorIntensity {type=NiagaraFloat, defaultValue=1, description=""}
  User.PaintedColor {type=LinearColor, description="", defaultValue {struct=LinearColor, value=<rgba 1 1 1 1>}}
  User.PaintedScalar {type=NiagaraFloat, defaultValue=1, description=""}
  User.PortCount {type=NiagaraInt32, defaultValue=2}
  User.RibbonCoreIntensity {type=NiagaraFloat, defaultValue=10, description=""}
  User.RibbonMidIntensity {type=NiagaraFloat, defaultValue=5, description=""}
  User.RibbonOutIntensity {type=NiagaraFloat, defaultValue=2, description=""}
  User.RingsIntensityB {type=NiagaraFloat, defaultValue=5, description=""}
  User.RingsIntensityG {type=NiagaraFloat, defaultValue=5, description=""}
  User.RingsIntensityR {type=NiagaraFloat, defaultValue=5, description=""}
  User.RingsOpacity {type=NiagaraFloat, defaultValue=1, description=""}
  User.SingleColorIntensity {type=NiagaraFloat, defaultValue=1, description=""}
 ]
}
