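# M_Cone is an unlit, masked Niagara mesh-particle material for the Pulsed Energy vehicle boost cosmetic, likely shading a cone mesh.
# It multiplies particle color RGB by the Intensity dynamic parameter, so the particle system drives hue and brightness.
# The product feeds the Emissive input of the shared MF_FX_EmissiveAdjustments function.
# The graph has no textures or mask logic of its own.
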Material /VehicleCosmetics/Boosts/Boost_PulsedEnergy/Materials/M_Cone
graph {
 properties {
  bAutomaticallySetUsageInEditor=false
  bUsedWithNiagaraMeshParticles=true
  blendMode=BLEND_Masked
  pixelDepthOffsetMode=PDOM_Legacy
  shadingModel=MSM_Unlit
  usedShadingModels=Unlit
 }
 nodes [MaterialFunctionCall(Emissive=Multiply(A=ParticleColor().RGB, B=DynamicParameter() {paramNames [Intensity,Param2,Param3,Param4]}.Intensity)) {materialFunction=&MF_FX_EmissiveAdjustments}]
}
