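# Masked, unlit, two-sided material for Niagara mesh particles, drawing the impact effect of the Pulsed Energy vehicle boost.
# A dynamic isPainted flag selects particle color scaled by two intensities or fixed purple tones, each blended by face side, feeding an emissive adjustment function.
# An impact gradient texture, boosted by one of two static-switched multipliers, combines with dynamic general and impact opacity values into a saturated mask.
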Material /VehicleCosmetics/Boosts/Boost_PulsedEnergy/Materials/M_Impact
graph {
 properties {
  bAutomaticallySetUsageInEditor=false
  bUsedWithNiagaraMeshParticles=true
  blendMode=BLEND_Masked
  pixelDepthOffsetMode=PDOM_Legacy
  shadingModel=MSM_Unlit
  twoSided=true
  usedShadingModels=Unlit
 }
 nodes [
  n1 = DynamicParameter() {paramNames [GeneralOpacity, ImpactOpacity, Color1Intensity, Color2Inensity]}
  n2 = ParticleColor()
  n3 = Multiply(A=@n2.RGB, B=@n1.Color1Intensity)
  n4 = ConstantBiasScale(None=TwoSidedSign())
  n5 = Multiply(A=@n1.Color2Inensity, B=@n2.RGB)
  n6 = LinearInterpolate(A=Constant3Vector() {constant=<rgba 0.3636 0.142 1 0>}, B=Constant3Vector() {constant=<rgba 0.0179 0 0.0869 0>}, Alpha=@n4)
  MaterialFunctionCall(Emissive=If(
    A=DynamicParameter() {parameterIndex=1, defaultValue=<rgba 0 0 0 0>, paramNames [isPainted?, Param2, Param3, Param4]}.isPainted?
    "A > B"=LinearInterpolate(A=@n3, B=@n5, Alpha=@n4)
    "A == B"=@n6
    "A < B"=@n6
   )) {materialFunction=&MF_FX_EmissiveAdjustments}
  NamedRerouteDeclaration(None=@n1.GeneralOpacity) {name=Dynamic-GeneralOpacity, nodeColor=<rgba 1 0 0.75 1>}
  NamedRerouteDeclaration(None=@n1.ImpactOpacity) {name=Dynamic-ImpactOpacity, nodeColor=<rgba 1 0 0.75 1>}
  n7 = NamedRerouteUsage()
  n8 = TextureSampleParameter2D(UVs=TextureCoordinate()) {parameterName=Opacity, texture=/DealershipSharedLibrary/FX/Textures/Gradient_Masks/T_VCVFX_ImpactFilled_01_A}
  n9 = Constant() {r=10}
  n10 = Constant() {r=5}
  n11 = StaticSwitchParameter(True=Saturate(None=Multiply(A=@n8.RGB, B=@n9)), False=Saturate(None=Multiply(A=@n8.RGB, B=@n10))) {parameterName=ImpactOpacity}
  Saturate(None=Multiply(A=NamedRerouteUsage(), B=Multiply(A=@n7, B=@n11)))
 ]
}
