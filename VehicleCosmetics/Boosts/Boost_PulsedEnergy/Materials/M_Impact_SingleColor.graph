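# Unlit, two-sided masked material for Niagara mesh particles that draws a single-color impact effect for the Pulsed Energy vehicle boost cosmetic.
# Particle color, scaled by the Color1Intensity dynamic parameter, feeds the MF_FX_EmissiveAdjustments function.
# An impact texture's RGB is boosted and saturated, with a static switch choosing between two boost strengths, then multiplied by dynamic-parameter opacity controls through named reroutes and saturated again.
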Material /VehicleCosmetics/Boosts/Boost_PulsedEnergy/Materials/M_Impact_SingleColor
graph {
 properties {
  bAutomaticallySetUsageInEditor=false
  bUsedWithNiagaraMeshParticles=true
  blendMode=BLEND_Masked
  pixelDepthOffsetMode=PDOM_Legacy
  shadingModel=MSM_Unlit
  twoSided=true
  usedShadingModels=Unlit
 }
 nodes [
  n1 = DynamicParameter() {paramNames [GeneralOpacity, ImpactOpacity, Color1Intensity, Param4]}
  MaterialFunctionCall(Emissive=Multiply(A=ParticleColor().RGB, B=@n1.Color1Intensity)) {materialFunction=&MF_FX_EmissiveAdjustments}
  NamedRerouteDeclaration(None=@n1.GeneralOpacity) {name=Dynamic-GeneralOpacity, nodeColor=<rgba 1 0 0.75 1>}
  NamedRerouteDeclaration(None=@n1.ImpactOpacity) {name=Dynamic-ImpactOpacity, nodeColor=<rgba 1 0 0.75 1>}
  n2 = NamedRerouteUsage()
  n3 = TextureSampleParameter2D(UVs=TextureCoordinate()) {parameterName=Opacity, texture=/DealershipSharedLibrary/FX/Textures/Gradient_Masks/T_VCVFX_ImpactFilled_01_A}
  n4 = Constant() {r=10}
  n5 = Constant() {r=5}
  n6 = StaticSwitchParameter(True=Saturate(None=Multiply(A=@n3.RGB, B=@n4)), False=Saturate(None=Multiply(A=@n3.RGB, B=@n5))) {parameterName=ImpactOpacity}
  Saturate(None=Multiply(A=NamedRerouteUsage(), B=Multiply(A=@n2, B=@n6)))
 ]
}
