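# Unlit, two-sided translucent trail material for the PulsedEnergy vehicle boost, used on Niagara ribbons and mesh particles.
# Panning caustics noise distorts the U coordinate of an RGB trail texture, whose channels act as separate masks for color and opacity.
# Dynamic parameters set color intensities and an isPainted value for switching between default and painted looks.
# Opacity fades along U, with depth fade skipped on low-end mobile and Switch quality tiers.
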Material /VehicleCosmetics/Boosts/Boost_PulsedEnergy/Materials/M_Trail
graph {
 properties {
  bAutomaticallySetUsageInEditor=false
  bUsedWithNiagaraMeshParticles=true
  bUsedWithNiagaraRibbons=true
  blendMode=BLEND_Translucent
  pixelDepthOffsetMode=PDOM_Legacy
  shadingModel=MSM_Unlit
  twoSided=true
  usedShadingModels=Unlit
 }
 nodes [
  n1 = NamedRerouteUsage()
  n2 = NamedRerouteUsage()
  n3 = DynamicParameter() {defaultValue=<rgba 1 1 1 0>, paramNames [Color1Intensity, Color2Intensity, Color3Intensity, isPainted?]}
  n4 = NamedRerouteUsage()
  n5 = Add(A=Multiply(A=Multiply(A=@n1, B=@n2), B=@n3.Color2Intensity), B=Multiply(A=Multiply(A=@n1, B=@n4), B=@n3.Color3Intensity))
  n6 = Multiply(A=Multiply(A=@n1, B=NamedRerouteUsage()), B=@n3.Color1Intensity)
  n7 = Multiply(A=VectorParameter() {parameterName=Default Color Blue, defaultValue=<rgba 0.0161 0.9914 0.9323 1>}.RGB, B=NamedRerouteUsage())
  n8 = NamedRerouteUsage()
  n9 = Multiply(A=VectorParameter() {parameterName=Default Color Pink, defaultValue=<rgba 0.3787 0.1604 0.9914 1>}.RGB, B=@n8)
  n10 = Constant() {r=3}
  n11 = NamedRerouteUsage()
  n12 = Multiply(A=NamedRerouteUsage(), B=Add(A=Add(A=@n7, B=@n9), B=Multiply(A=@n10, B=@n11)))
  MaterialFunctionCall(Emissive=If(A=NamedRerouteUsage(), "A > B"=Add(A=@n5, B=@n6), "A == B"=@n12, "A < B"=@n12)) {materialFunction=&MF_FX_EmissiveAdjustments}
  n13 = NamedRerouteUsage()
  n14 = ComponentMask(None=TextureCoordinate()) {r=true}
  n15 = Multiply(A=@n14, B=Constant() {r=6})
  n16 = ScalarParameter() {defaultValue=0.5, parameterName=PowerFade}
  n17 = OneMinus(None=@n14)
  n18 = StaticSwitchParameter(True=Constant() {r=1}, False=Constant() {r=0.8}) {parameterName=isPainted?}
  n19 = Multiply(A=Add(A=Add(A=@n11, B=@n13), B=@n8), B=Multiply(A=Power(Base=@n15, Exp=@n16), B=Power(Base=@n17, Exp=@n18)))
  n20 = NamedRerouteUsage()
  n21 = If(A=NamedRerouteUsage(), "A > B"=Saturate(None=Multiply(A=@n19, B=@n20)), "A == B"=@n19, "A < B"=@n19)
  n22 = Multiply(
   A=@n21
   B=MaterialFunctionCall(Opacity In=Constant() {r=1}, DepthFade Distance=ScalarParameter() {defaultValue=50, parameterName=FadeDistance}) {materialFunction=&MF_VCVFX_DepthFade}.Result
  )
  MaterialFunctionCall!(
   Low _ PC Low and Med=@n22
   "High _ XB1, PS4"=@n22
   "Epic _ PC High, XSX, PS5"=@n22
   --------
   "Mobile Low _ PC Perf, Android"=@n21
   Mobile Med _ Switch=@n21
  ) {materialFunction=/Game/Global/MaterialFunctions/TechArt/MF_Fortnite_QualitySwitch_Scalar}
  n23 = ParticleColor()
  NamedRerouteDeclaration(None=@n23.RGB) {name=ParticleColor-RGB, nodeColor=<rgba 1 0 0.75 1>}
  NamedRerouteDeclaration(None=@n23.A) {name=ParticleColor-A, nodeColor=<rgba 1 0 0.75 1>}
  n24 = TextureCoordinate()
  n25 = Panner(Coordinate=TextureCoordinate() {uTiling=2, vTiling=2}, Time=Multiply(A=Time(), B=ScalarParameter() {defaultValue=1, parameterName=Time})) {speedX=-1}
  n26 = AppendVector(
   A=Multiply(A=TextureSample(UVs=@n25) {texture=/DealershipSharedLibrary/FX/Textures/Noise/T_VCVFX_CausticsNoise_A}.R, B=Constant() {r=0.35})
   B=Constant()
  )
  n27 = TextureSampleParameter2D(UVs=Saturate(None=Add(A=@n24, B=@n26))) {
   parameterName=Texture
   samplerSource=SSM_Clamp_WorldGroupSettings
   texture=/DealershipSharedLibrary/FX/Textures/TexturePackage/T_VCVFX_Trail_RGB
  }
  NamedRerouteDeclaration(None=@n27.R) {name=Texture-R, nodeColor=<rgba 1 0 0.75 1>}
  NamedRerouteDeclaration(None=@n27.G) {name=Texture-G, nodeColor=<rgba 1 0 0.9375 1>}
  NamedRerouteDeclaration(None=@n27.B) {name=Texture-B, nodeColor=<rgba 1 0 0.75 1>}
  NamedRerouteDeclaration(None=@n3.isPainted?) {name=Dynamic-isPainted?, nodeColor=<rgba 1 0 0.75 1>}
 ]
}
