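# Unlit, translucent, two-sided material for Niagara mesh particles, named for white rings in the Pulsed Energy vehicle boost cosmetic.
# Squared particle color passes through an emissive adjustment function, with no textures involved.
# Opacity is squared particle alpha scaled by an Opacity dynamic parameter and saturated.
# A quality switch applies depth fade on PC and console tiers but skips it on both mobile tiers.
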Material /VehicleCosmetics/Boosts/Boost_PulsedEnergy/Materials/M_WhiteRings
graph {
 properties {
  bAutomaticallySetUsageInEditor=false
  bUsedWithNiagaraMeshParticles=true
  blendMode=BLEND_Translucent
  pixelDepthOffsetMode=PDOM_Legacy
  shadingModel=MSM_Unlit
  twoSided=true
  usedShadingModels=Unlit
 }
 nodes [
  n1 = ParticleColor()
  n2 = Saturate(None=Multiply(A=Multiply(A=@n1.A, B=@n1.A), B=DynamicParameter() {paramNames [Opacity,Param2,Param3,Param4]}.Opacity))
  n3 = Multiply(
   A=@n2
   B=MaterialFunctionCall(Opacity In=Constant() {r=1}, DepthFade Distance=ScalarParameter() {defaultValue=50, parameterName=FadeDiistance, sliderMax=50}) {materialFunction=&MF_VCVFX_DepthFade}.Result
  )
  MaterialFunctionCall!(
   Low _ PC Low and Med=@n3
   "High _ XB1, PS4"=@n3
   "Epic _ PC High, XSX, PS5"=@n3
   --------
   "Mobile Low _ PC Perf, Android"=@n2
   Mobile Med _ Switch=@n2
  ) {materialFunction=/Game/Global/MaterialFunctions/TechArt/MF_Fortnite_QualitySwitch_Scalar}
  MaterialFunctionCall(Emissive=Multiply(A=@n1.RGB, B=@n1.RGB)) {materialFunction=&MF_FX_EmissiveAdjustments}
 ]
}
