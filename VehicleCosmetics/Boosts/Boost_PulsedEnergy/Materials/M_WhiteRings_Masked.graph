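# Unlit, two-sided masked material for Niagara mesh particles in the Pulsed Energy vehicle boost effect, named for white rings.
# Squares particle color RGB and passes it through MF_FX_EmissiveAdjustments for the emissive.
# Squares particle alpha, scales it by the Opacity dynamic parameter and saturates it for the mask.
# It samples no textures in this graph, so particle inputs drive the look.
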Material /VehicleCosmetics/Boosts/Boost_PulsedEnergy/Materials/M_WhiteRings_Masked
graph {
 properties {
  bAutomaticallySetUsageInEditor=false
  bUsedWithNiagaraMeshParticles=true
  blendMode=BLEND_Masked
  pixelDepthOffsetMode=PDOM_Legacy
  shadingModel=MSM_Unlit
  twoSided=true
  usedShadingModels=Unlit
 }
 nodes [
  n1 = ParticleColor()
  MaterialFunctionCall(Emissive=Multiply(A=@n1.RGB, B=@n1.RGB)) {materialFunction=&MF_FX_EmissiveAdjustments}
  Saturate(None=Multiply(A=Multiply(A=@n1.A, B=@n1.A), B=DynamicParameter() {paramNames [Opacity,Param2,Param3,Param4]}.Opacity))
 ]
}
